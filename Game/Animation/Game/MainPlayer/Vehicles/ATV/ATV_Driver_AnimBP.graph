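# ATV driver Animation Blueprint for the player skeleton, supplying the VehicleFinalPoseOverride layer; its other layers and EventGraph are empty.
# A locomotion state machine handles steering, starts, stops, pivots and boosting, then gains head aim, additive collision, jump, flip and landing states, and body drag.
# Montage slots and a gender-based additive follow.
# Finally, curve-gated world-space targets move the IK hand bones, and Leg IK nodes solve both arms.
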AnimBlueprint /Game/Animation/Game/MainPlayer/Vehicles/ATV/ATV_Driver_AnimBP
target_skeleton=&Fortnite_M_Avg_Player_Skeleton
graphs [
 AnimGraph {nodes [Output Pose() {layerGroup=None}]}
 EventGraph {links [], nodes []}
 VehicleFinalPoseOverride {nodes [
   Input Pose() {name=InFinalPose}
   n1 = "Blend Poses (EFortCustomGender)"!(
    BlendPose_0:Pose Link Structure=Additive Identity Pose()
    BlendPose_1:Pose Link Structure="Sequence Player (additive) 'ATV_Driver_Idle_Pose_CMF'"() {sequence=&/ATV_Driver_Idle_Pose_CMF}
    BlendTime_0:float=0.1
    BlendTime_1:float=0.1
    ActiveEnumValue:Byte=Get Gender() -> (Gender:EFortCustomGender Enum=Invalid)
   ) -> (Pose:Pose Link Structure) {
    is=title AnimGraphNode_BlendListByEnum
    blendType=Linear
    blendPose [{linkId=-1, sourceLinkId=-1}, {linkId=-1, sourceLinkId=-1}]
    blendTime [0.1,0.1]
   }
   n2 = Use cached pose 'PreUpperBody'()
   n3 = Apply Additive(
    Base=Layered blend per bone(BasePose=Use cached pose 'PreUpperBody'(), BlendPoses_0=Slot 'UpperBody'(Source=@n2) {slotName=UpperBody}) {layerSetup [{branchFilters [{boneName=pelvis}, {blendDepth=-1, boneName=hand_l}]}]}
    Additive=@n1
   ) {alphaBoolBlend {blendOption=Linear}}
   n4 = "Transform (Modify) Bone - Bone: ik_hand_l"!(
    ComponentPose=Local To Component(LocalPose=Slot 'FullBody'(Source=@n3) {slotName=FullBody})
    AlphaCurveName:Name=LeftHandFK
    Translation=Get WorldPosLeft() -> (WorldPosLeft:Vector)
    Rotation=Get WorldRotLeft() -> (WorldRotLeft:Rotator)
   ) -> (Pose) {
    alphaInputType=Curve
    rotationMode=BMM_Replace
    rotationSpace=BCS_WorldSpace
    translationMode=BMM_Replace
    translationSpace=BCS_WorldSpace
    alphaBoolBlend {blendOption=Linear}
    alphaScaleBiasClamp {bias=1, scale=-1}
    boneToModify {boneName=ik_hand_l}
   }
   n5 = "Transform (Modify) Bone - Bone: ik_hand_r"!(
    ComponentPose=@n4
    AlphaCurveName:Name=RightHandFK
    Translation=Get WorldPosRight() -> (WorldPosRight:Vector)
    Rotation=Get WorldRotRight() -> (WorldRotRight:Rotator)
   ) -> (Pose) {
    alphaInputType=Curve
    rotationMode=BMM_Replace
    rotationSpace=BCS_WorldSpace
    translationMode=BMM_Replace
    translationSpace=BCS_WorldSpace
    alphaBoolBlend {blendOption=Linear}
    alphaScaleBiasClamp {bias=1, scale=-1}
    boneToModify {boneName=ik_hand_r}
   }
   n6 = Leg IK!(ComponentPose:Component Space Pose Link Structure=@n5, Alpha:float=Get RightHandIKAlpha() -> (RightHandIKAlpha:float)) -> (Pose:Component Space Pose Link Structure) {
    is=title AnimGraphNode_LegIK
    alphaBoolBlend {blendOption=Linear}
    legsDefinition [{
      bEnableKneeTwistCorrection=false
      bEnableRotationLimit=false
      footBoneForwardAxis=Y
      hingeRotationAxis=Z
      minRotationAngle=15
      numBonesInLimb=2
      twistOffsetCurveName=None
      fKFootBone {boneName=hand_r}
      iKFootBone {boneName=ik_hand_r}
     }]
   }
   n7 = Leg IK(ComponentPose=@n6, Alpha=Get LeftHandIKAlpha() -> (LeftHandIKAlpha:float)) {
    alphaBoolBlend {blendOption=Linear}
    legsDefinition [{
      bEnableKneeTwistCorrection=false
      bEnableRotationLimit=false
      footBoneForwardAxis=Y
      hingeRotationAxis=Z
      minRotationAngle=15
      numBonesInLimb=2
      twistOffsetCurveName=None
      fKFootBone {boneName=hand_l}
      iKFootBone {boneName=ik_hand_l}
     }]
   }
   Output Pose(Result=Component To Local(ComponentPose=@n7)) {layerGroup=VehicleOverride}
   Save cached pose 'JumpLeanPose'(Pose=Blendspace Player 'ATV_Driver_JumpLeans_CMM_BS'!(X=Get LocalVelocityYawAngle() -> (LocalVelocityYawAngle:float)) -> (Pose) {blendSpace=&/ATV_Driver_JumpLeans_CMM_BS}) {cacheName=JumpLeanPose}
   n8 = "Blend Poses (EFortCardinalDirection)"!(
    BlendPose_0:Pose Link Structure=Sequence Player 'ATV_Driver_Steer_N_CMM'() {sequence=&/ATV_Driver_Steer_N_CMM}
    BlendPose_1:Pose Link Structure=Blend Poses by bool(
     BlendPose_0=Sequence Player 'ATV_Driver_Steer_E_CMM'() {sequence=&/ATV_Driver_Steer_E_CMM}
     BlendPose_1=Sequence Player 'ATV_Driver_Steer_E_Fast_CMM'() {sequence=&/ATV_Driver_Steer_E_Fast_CMM}
     BlendTime_0=0.4
     BlendTime_1=0.4
     bActiveValue=Get bIsTurningRightHard() -> (bIsTurningRightHard:bool)
    )
    BlendPose_2:Pose Link Structure=Sequence Player 'ATV_Driver_Steer_S_CMM'() {sequence=&/ATV_Driver_Steer_S_CMM}
    BlendPose_3:Pose Link Structure=Blend Poses by bool(
     BlendPose_0=Sequence Player 'ATV_Driver_Steer_W_CMM'() {sequence=&/ATV_Driver_Steer_W_CMM}
     BlendPose_1=Sequence Player 'ATV_Driver_Steer_W_Fast_CMM'() {sequence=&/ATV_Driver_Steer_W_Fast_CMM}
     BlendTime_0=0.4
     BlendTime_1=0.4
     bActiveValue=Get bIsTurningLeftHard() -> (bIsTurningLeftHard:bool)
    )
    BlendTime_0:float=0.3
    BlendTime_1:float=0.3
    BlendTime_2:float=0.3
    BlendTime_3:float=0.3
    ActiveEnumValue:Byte=Get LocomotionCardinalDirection() -> (LocomotionCardinalDirection:EFortCardinalDirection Enum=North)
   ) -> (Pose:Pose Link Structure) {
    is=title AnimGraphNode_BlendListByEnum
    blendPose [{linkId=-1, sourceLinkId=-1}, {linkId=-1, sourceLinkId=-1}, {linkId=-1, sourceLinkId=-1}, {linkId=-1, sourceLinkId=-1}]
    blendTime [0.1,0.1,0.1,0.1]
   }
   Save cached pose 'DriverSteerDirectionPose'(Pose=@n8) {cacheName=DriverSteerDirectionPose}
   AnimGraphNode_StateMachine_0 = Locomotion() {is=title AnimGraphNode_StateMachine}
   n9 = AimOffset Player 'ATV_Driver_Head_AO'!(
    X:float=Get AimYaw() -> (AimYaw:float)
    Y:float=Get AimPitch()
    BasePose:Pose Link Structure=@AnimGraphNode_StateMachine_0
    bAlphaBoolEnabled:bool=Get bShouldEnableAO() -> (bShouldEnableAO:bool)
   ) -> (Pose:Pose Link Structure) {
    alphaInputType=Bool
    blendSpace=/Game/Animation/Game/MainPlayer/Vehicles/ATV/AO/ATV_Driver_Head_AO
    alphaBoolBlend {blendInTime=0.2, blendOption=Linear, blendOutTime=0.1}
   }
   AnimGraphNode_StateMachine_1 = FullBodyAdditive() {is=title AnimGraphNode_StateMachine}
   n10 = Break Fort Anim Input Quad!(FortAnimInput_Quad:"Fort Anim Input Quad Structure (by ref)"=Get QuadBikeInput() -> (QuadBikeInput:Fort Anim Input Quad Structure)) -> (
    bIsUsingQuad:bool=~
    bIsDriver:bool=~
    bIsFrontPassenger:bool=~
    bIsBackPassenger:bool=~
    bIsBackPassengerAndLeaning:bool=~
    bIsDrifting:bool=~
    bIsBoosting:bool=~
    bIsReversing:bool=~
    bIsBraking:bool=~
    bIsMoving:bool=~
    bIsMovingForward:bool=~
    bIsLeaning:bool=~
    bIsLeaningOrBouncing:bool=~
    bIsBounceCrouching:bool=~
    bIsBounceCrouched:bool=~
    bIsBounceJumping:bool=~
    bIsBounceRecoiling:bool=~
    bIsSteeringRight:bool=~
    bIsSteeringLeft:bool=~
    RunForwardAlpha:float=~
    BounceCompression:float=~
    LeanPosition:Vector=~
    LeanPositionX:float=~
    LeanPositionY:float=~
    LeanPositionZ:float=~
    VerticalVelocity:float=~
    VerticalAcceleration:float=~
    bAimFWD:bool=~
    bAimBWD:bool=~
    bAimLFT:bool=~
    bAimRGT:bool=~
    PawnToVehicleDeltaYawAngleDegrees:float=~
    AimCardDirDeadZoneAngleDegrees:float=~
    AimCardDirAngleOffsetDegrees:float=~
    AimFWDDeltaAngleDegrees:float=~
    AimBWDDeltaAngleDegrees:float=~
    AimLFTDeltaAngleDegrees:float=~
    AimRGTDeltaAngleDegrees:float=~
    SlopePitchDegreeAngle:float=~
    SlopeRollDegreeAngle:float=~
    SteerAngle:float=~
    SteerAlpha:float=~
    SteerAngleDeadZoneDegrees:float=~
    SteeringRotation:float=~
    VehiclePitch:float=~
    VehicleRoll:float=~
   ) {is=title K2Node_BreakStruct}
   Save cached pose 'PreUpperBody'(Pose=Apply Additive!(
     Base=Apply Additive!(Base=@n9, Additive=@AnimGraphNode_StateMachine_1, Alpha:float) -> (Pose) {alphaBoolBlend {blendOption=Linear}}
     Additive=Blendspace Player 'ATV_Driver_BodyDrag_BS_CMM'!(X=@n10.SteeringRotation) -> (Pose) {blendSpace=&/ATV_Driver_BodyDrag_BS_CMM}
     bAlphaBoolEnabled:bool=Get bIsBodyDragging() -> (bIsBodyDragging:bool)
    ) -> (Pose) {alphaInputType=Bool, alphaBoolBlend {blendInTime=0.25, blendOption=Linear, blendOutTime=0.25}}) {cacheName=PreUpperBody}
  ]}
 Locomotion {
  entry=Idle
  path=VehicleFinalPoseOverride.AnimGraphNode_StateMachine_0.Locomotion
  ids=id
  nodes [
   AnimStateNode_0 = Idle()
   AnimStateNode_1 = MovementPivot() {is=title AnimStateNode}
   AnimStateNode_4 = MovementStop() {is=title AnimStateNode}
   AnimStateNode_5 = MovementStart() {is=title AnimStateNode}
   AnimStateNode_6 = MovementLoop() {is=title AnimStateNode}
   AnimStateTransitionNode_0 = AnimStateTransitionNode() {crossfadeDuration=0.4}
   AnimStateTransitionNode_1 = AnimStateTransitionNode() {bAutomaticRuleBasedOnSequencePlayerInState=true, crossfadeDuration=0.5}
   AnimStateTransitionNode_10 = AnimStateTransitionNode()
   AnimStateTransitionNode_4 = AnimStateTransitionNode() {bAutomaticRuleBasedOnSequencePlayerInState=true, crossfadeDuration=0.4}
   AnimStateTransitionNode_5 = AnimStateTransitionNode()
   AnimStateTransitionNode_7 = AnimStateTransitionNode() {priorityOrder=2}
   AnimStateTransitionNode_8 = AnimStateTransitionNode() {bAutomaticRuleBasedOnSequencePlayerInState=true, crossfadeDuration=0.4}
   AnimStateTransitionNode_9 = AnimStateTransitionNode()
  ]
  links [
   {
    dst=MovementStart
    rule=VehicleFinalPoseOverride.AnimGraphNode_StateMachine_0.Locomotion.AnimStateTransitionNode_9.Transition
    src=Idle
    via=AnimStateTransitionNode_9
   }
   {
    dst=Idle
    rule=VehicleFinalPoseOverride.AnimGraphNode_StateMachine_0.Locomotion.AnimStateTransitionNode_5.Transition
    src=MovementLoop
    via=AnimStateTransitionNode_5
   }
   {
    dst=MovementPivot
    rule=VehicleFinalPoseOverride.AnimGraphNode_StateMachine_0.Locomotion.AnimStateTransitionNode_0.Transition
    src=MovementLoop
    via=AnimStateTransitionNode_0
   }
   {
    dst=MovementStop
    rule=VehicleFinalPoseOverride.AnimGraphNode_StateMachine_0.Locomotion.AnimStateTransitionNode_10.Transition
    src=MovementLoop
    via=AnimStateTransitionNode_10
   }
   {
    dst=MovementStop
    rule=VehicleFinalPoseOverride.AnimGraphNode_StateMachine_0.Locomotion.AnimStateTransitionNode_7.Transition
    src=MovementLoop
    via=AnimStateTransitionNode_7
   }
   {
    dst=MovementLoop
    rule=VehicleFinalPoseOverride.AnimGraphNode_StateMachine_0.Locomotion.AnimStateTransitionNode_1.Transition
    src=MovementPivot
    via=AnimStateTransitionNode_1
   }
   {
    dst=MovementLoop
    rule=VehicleFinalPoseOverride.AnimGraphNode_StateMachine_0.Locomotion.AnimStateTransitionNode_4.Transition
    src=MovementStart
    via=AnimStateTransitionNode_4
   }
   {
    dst=Idle
    rule=VehicleFinalPoseOverride.AnimGraphNode_StateMachine_0.Locomotion.AnimStateTransitionNode_8.Transition
    src=MovementStop
    via=AnimStateTransitionNode_8
   }
  ]
 }
 Idle {
  path=VehicleFinalPoseOverride.AnimGraphNode_StateMachine_0.Locomotion.AnimStateNode_0.Idle
  nodes [Output Animation Pose(Result=Blend Poses by bool(
     BlendPose_0=Use cached pose 'DriverSteerDirectionPose'()
     BlendPose_1=Sequence Player 'ATV_Driver_Idle_Pose_CMM'() {sequence=&/ATV_Driver_Idle_Pose_CMM}
     BlendTime_0=0.35
     BlendTime_1=0.25
     bActiveValue=Get bIsSteeringRotating() -> (bIsSteeringRotating:bool)
    ) {blendType=Linear})]
 }
 MovementPivot {
  path=VehicleFinalPoseOverride.AnimGraphNode_StateMachine_0.Locomotion.AnimStateNode_1.MovementPivot
  nodes [
   n1 = "Blend Poses (EFortCardinalDirection)"(
    BlendPose_0=Sequence Player 'ATV_Drive_Pivot_StoN'() {bLoopAnimation=false, sequence=&/ATV_Drive_Pivot_StoN}
    BlendPose_1=Sequence Player 'ATV_Drive_Pivot_NtoS'() {bLoopAnimation=false, sequence=&/ATV_Drive_Pivot_NtoS}
    BlendPose_2=Sequence Player 'ATV_Drive_Pivot_E'!(PlayRate:float=1) -> (Pose) {bLoopAnimation=false, sequence=&/ATV_Drive_Pivot_E}
    BlendPose_3=Sequence Player 'ATV_Drive_Pivot_W'!(PlayRate:float=1) -> (Pose) {bLoopAnimation=false, sequence=&/ATV_Drive_Pivot_W}
    BlendTime_0=0.3
    BlendTime_1=0.3
    BlendTime_2=0.4
    BlendTime_3=0.4
    ActiveEnumValue=Get PrePivotCardinalDirection() -> (PrePivotCardinalDirection:EFortCardinalDirection Enum=North)
   ) {
    blendPose [{linkId=-1, sourceLinkId=-1}, {linkId=-1, sourceLinkId=-1}, {linkId=-1, sourceLinkId=-1}, {linkId=-1, sourceLinkId=-1}]
    blendTime [0.1,0.1,0.1,0.1]
   }
   Output Animation Pose(Result=@n1)
  ]
 }
 MovementStop {
  path=VehicleFinalPoseOverride.AnimGraphNode_StateMachine_0.Locomotion.AnimStateNode_4.MovementStop
  nodes [
   n1 = "Blend Poses (EFortCardinalDirection)"!(
    BlendPose_0=Sequence Player 'ATV_Drive_N_Stop_CMM'() {bLoopAnimation=false, sequence=&/ATV_Drive_N_Stop_CMM}
    BlendPose_1=Sequence Player 'ATV_Driver_Reverse_Stop_CMM'() {bLoopAnimation=false, sequence=&/ATV_Driver_Reverse_Stop_CMM}
    BlendTime_0=0.3
    BlendTime_1=0.4
    ActiveEnumValue=Get LocomotionCardinalDirection() -> (LocomotionCardinalDirection=North)
   ) -> (Pose) {blendType=Linear, blendPose [{linkId=-1, sourceLinkId=-1}, {linkId=-1, sourceLinkId=-1}], blendTime [0.1,0.1]}
   Output Animation Pose(Result=@n1)
  ]
 }
 MovementStart {
  path=VehicleFinalPoseOverride.AnimGraphNode_StateMachine_0.Locomotion.AnimStateNode_5.MovementStart
  nodes [
   n1 = Apply Additive(
    Base=Use cached pose 'DriverSteerDirectionPose'()
    Additive="Sequence Player (additive) 'ATV_Drive_N_Start_CMM'"() {bLoopAnimation=false, sequence=&/ATV_Drive_N_Start_CMM}
    Alpha=Get AbsSteerAngle() -> (AbsSteerAngle:float)
   ) {
    alphaBoolBlend {blendOption=Linear}
    alphaScaleBiasClamp {bInterpResult=true, bMapRange=true, interpSpeedIncreasing=2, inRange {max=30, min=14}, outRange {max=0.25, min=1}}
   }
   Output Animation Pose(Result="Blend Poses (EFortCardinalDirection)"(
     BlendPose_0=@n1
     BlendPose_1=Sequence Player 'ATV_Driver_Reverse_Start_CMM'() {bLoopAnimation=false, sequence=&/ATV_Driver_Reverse_Start_CMM}
     BlendTime_0=0.1
     BlendTime_1=0.1
     ActiveEnumValue=Get LocomotionCardinalDirection() -> (LocomotionCardinalDirection=North)
    ) {blendPose [{linkId=-1, sourceLinkId=-1}, {linkId=-1, sourceLinkId=-1}], blendTime [0.1,0.1]})
  ]
 }
 MovementLoop {
  path=VehicleFinalPoseOverride.AnimGraphNode_StateMachine_0.Locomotion.AnimStateNode_6.MovementLoop
  nodes [
   AnimGraphNode_StateMachine_0 = Boosting() {is=title AnimGraphNode_StateMachine}
   Output Animation Pose(Result=Apply Additive!(
     Base=Use cached pose 'DriverSteerDirectionPose'()
     Additive=@AnimGraphNode_StateMachine_0
     bAlphaBoolEnabled:bool=Get bIsOnGroundAndBoosting() -> (bIsOnGroundAndBoosting:bool)
    ) -> (Pose) {alphaInputType=Bool, alphaBoolBlend {blendInTime=0.2, blendOutTime=0.3}})
  ]
 }
 Boosting {
  entry=BoostingStart
  path=VehicleFinalPoseOverride.AnimGraphNode_StateMachine_0.Locomotion.AnimStateNode_6.MovementLoop.AnimGraphNode_StateMachine_0.Boosting
  ids=id
  nodes [
   AnimStateNode_0 = BoostingStart() {is=title AnimStateNode}
   AnimStateNode_1 = BoostingPose() {is=title AnimStateNode}
   AnimStateTransitionNode_0 = AnimStateTransitionNode() {bAutomaticRuleBasedOnSequencePlayerInState=true}
  ]
  links [{
    dst=BoostingPose
    rule=VehicleFinalPoseOverride.AnimGraphNode_StateMachine_0.Locomotion.AnimStateNode_6.MovementLoop.AnimGraphNode_StateMachine_0.Boosting.AnimStateTransitionNode_0.Transition
    src=BoostingStart
    via=AnimStateTransitionNode_0
   }]
 }
 BoostingStart {
  path=VehicleFinalPoseOverride.AnimGraphNode_StateMachine_0.Locomotion.AnimStateNode_6.MovementLoop.AnimGraphNode_StateMachine_0.Boosting.AnimStateNode_0.BoostingStart
  nodes [Output Animation Pose(Result="Sequence Player (additive) 'ATV_Driver_Boosting_Start_CMM'"() {bLoopAnimation=false, sequence=&/ATV_Driver_Boosting_Start_CMM})]
 }
 BoostingPose {
  path=VehicleFinalPoseOverride.AnimGraphNode_StateMachine_0.Locomotion.AnimStateNode_6.MovementLoop.AnimGraphNode_StateMachine_0.Boosting.AnimStateNode_1.BoostingPose
  nodes [Output Animation Pose(Result="Sequence Player (additive) 'ATV_Driver_Boosting_Pose_CMM'"() {sequence=&/ATV_Driver_Boosting_Pose_CMM})]
 }
 Transition {
  path=VehicleFinalPoseOverride.AnimGraphNode_StateMachine_0.Locomotion.AnimStateNode_6.MovementLoop.AnimGraphNode_StateMachine_0.Boosting.AnimStateTransitionNode_0.Transition
  nodes [Result()]
 }
 Transition {
  path=VehicleFinalPoseOverride.AnimGraphNode_StateMachine_0.Locomotion.AnimStateTransitionNode_0.Transition
  nodes [Result(bCanEnterTransition=Get bPlayPivotOnGroundAndNotBoosting() -> (bPlayPivotOnGroundAndNotBoosting:bool))]
 }
 Transition {path=VehicleFinalPoseOverride.AnimGraphNode_StateMachine_0.Locomotion.AnimStateTransitionNode_1.Transition, nodes [Result()]}
 Transition {
  path=VehicleFinalPoseOverride.AnimGraphNode_StateMachine_0.Locomotion.AnimStateTransitionNode_10.Transition
  nodes [Result(bCanEnterTransition=Get bIsSlowingDown() -> (bIsSlowingDown:bool))]
 }
 Transition {path=VehicleFinalPoseOverride.AnimGraphNode_StateMachine_0.Locomotion.AnimStateTransitionNode_4.Transition, nodes [Result()]}
 Transition {
  path=VehicleFinalPoseOverride.AnimGraphNode_StateMachine_0.Locomotion.AnimStateTransitionNode_5.Transition
  nodes [Result(bCanEnterTransition=Get bIsStopped() -> (bIsStopped:bool))]
 }
 Transition {
  path=VehicleFinalPoseOverride.AnimGraphNode_StateMachine_0.Locomotion.AnimStateTransitionNode_7.Transition
  nodes [Result(bCanEnterTransition=Get bIsBraking() -> (bIsBraking:bool))]
 }
 Transition {
  path=VehicleFinalPoseOverride.AnimGraphNode_StateMachine_0.Locomotion.AnimStateTransitionNode_8.Transition
  nodes [
   Result()
   NOT Boolean(A=Break Fort Anim Input Quad(FortAnimInput_Quad=Get QuadBikeInput()) -> (
     bIsUsingQuad=~
     bIsDriver=~
     bIsFrontPassenger=~
     bIsBackPassenger=~
     bIsBackPassengerAndLeaning=~
     bIsDrifting=~
     bIsBoosting=~
     bIsReversing=~
     bIsBraking=~
     bIsMoving=~
     bIsMovingForward=~
     bIsLeaning=~
     bIsLeaningOrBouncing=~
     bIsBounceCrouching=~
     bIsBounceCrouched=~
     bIsBounceJumping=~
     bIsBounceRecoiling=~
     bIsSteeringRight=~
     bIsSteeringLeft=~
     RunForwardAlpha=~
     BounceCompression=~
     LeanPosition=~
     LeanPositionX=~
     LeanPositionY=~
     LeanPositionZ=~
     VerticalVelocity=~
     VerticalAcceleration=~
     bAimFWD=~
     bAimBWD=~
     bAimLFT=~
     bAimRGT=~
     PawnToVehicleDeltaYawAngleDegrees=~
     AimCardDirDeadZoneAngleDegrees=~
     AimCardDirAngleOffsetDegrees=~
     AimFWDDeltaAngleDegrees=~
     AimBWDDeltaAngleDegrees=~
     AimLFTDeltaAngleDegrees=~
     AimRGTDeltaAngleDegrees=~
     SlopePitchDegreeAngle=~
     SlopeRollDegreeAngle=~
     SteerAngle=~
     SteerAlpha=~
     SteerAngleDeadZoneDegrees=~
     SteeringRotation=~
     VehiclePitch=~
     VehicleRoll=~
    ).bIsMoving)
  ]
 }
 Transition {
  path=VehicleFinalPoseOverride.AnimGraphNode_StateMachine_0.Locomotion.AnimStateTransitionNode_9.Transition
  nodes [Result(bCanEnterTransition=Get bIsDriveStarting() -> (bIsDriveStarting:bool))]
 }
 FullBodyAdditive {
  entry=Default
  path=VehicleFinalPoseOverride.AnimGraphNode_StateMachine_1.FullBodyAdditive
  ids=id
  nodes [
   AnimStateNode_1 = Collision_N() {is=title AnimStateNode}
   AnimStateNode_2 = Collision_S() {is=title AnimStateNode}
   AnimStateNode_3 = JumpLand() {is=title AnimStateNode}
   AnimStateNode_4 = Default() {is=title AnimStateNode}
   AnimStateNode_5 = Jump() {is=title AnimStateNode, stateEntered {guid=#1, notifyName=JumpEntered}}
   AnimStateTransitionNode_10 = AnimStateTransitionNode() {bAutomaticRuleBasedOnSequencePlayerInState=true, crossfadeDuration=0}
   AnimStateTransitionNode_11 = AnimStateTransitionNode() {crossfadeDuration=0}
   AnimStateTransitionNode_4 = AnimStateTransitionNode() {bAutomaticRuleBasedOnSequencePlayerInState=true}
   AnimStateTransitionNode_5 = AnimStateTransitionNode() {crossfadeDuration=0.15}
   AnimStateTransitionNode_6 = AnimStateTransitionNode()
   AnimStateTransitionNode_7 = AnimStateTransitionNode() {crossfadeDuration=0}
   AnimStateTransitionNode_8 = AnimStateTransitionNode() {bAutomaticRuleBasedOnSequencePlayerInState=true, crossfadeDuration=0}
   AnimStateTransitionNode_9 = AnimStateTransitionNode() {crossfadeDuration=0.067}
  ]
  links [
   {
    dst=Default
    rule=VehicleFinalPoseOverride.AnimGraphNode_StateMachine_1.FullBodyAdditive.AnimStateTransitionNode_10.Transition
    src=Collision_N
    via=AnimStateTransitionNode_10
   }
   {
    dst=Default
    rule=VehicleFinalPoseOverride.AnimGraphNode_StateMachine_1.FullBodyAdditive.AnimStateTransitionNode_8.Transition
    src=Collision_S
    via=AnimStateTransitionNode_8
   }
   {
    dst=Collision_N
    rule=VehicleFinalPoseOverride.AnimGraphNode_StateMachine_1.FullBodyAdditive.AnimStateTransitionNode_11.Transition
    src=Default
    via=AnimStateTransitionNode_11
   }
   {
    dst=Collision_S
    rule=VehicleFinalPoseOverride.AnimGraphNode_StateMachine_1.FullBodyAdditive.AnimStateTransitionNode_7.Transition
    src=Default
    via=AnimStateTransitionNode_7
   }
   {
    dst=Jump
    rule=VehicleFinalPoseOverride.AnimGraphNode_StateMachine_1.FullBodyAdditive.AnimStateTransitionNode_5.Transition
    src=Default
    via=AnimStateTransitionNode_5
   }
   {
    dst=JumpLand
    rule=VehicleFinalPoseOverride.AnimGraphNode_StateMachine_1.FullBodyAdditive.AnimStateTransitionNode_9.Transition
    src=Jump
    via=AnimStateTransitionNode_9
   }
   {
    dst=Default
    rule=VehicleFinalPoseOverride.AnimGraphNode_StateMachine_1.FullBodyAdditive.AnimStateTransitionNode_4.Transition
    src=JumpLand
    via=AnimStateTransitionNode_4
   }
   {
    dst=Default
    rule=VehicleFinalPoseOverride.AnimGraphNode_StateMachine_1.FullBodyAdditive.AnimStateTransitionNode_6.Transition
    src=JumpLand
    via=AnimStateTransitionNode_6
   }
  ]
 }
 Collision_N {
  path=VehicleFinalPoseOverride.AnimGraphNode_StateMachine_1.FullBodyAdditive.AnimStateNode_1.Collision_N
  nodes [Output Animation Pose(Result="Sequence Player (additive) 'ATV_Driver_Collision_N_CMM'"() {bLoopAnimation=false, sequence=&/ATV_Driver_Collision_N_CMM})]
 }
 Collision_S {
  path=VehicleFinalPoseOverride.AnimGraphNode_StateMachine_1.FullBodyAdditive.AnimStateNode_2.Collision_S
  nodes [Output Animation Pose(Result="Sequence Player (additive) 'ATV_Driver_Collision_S_CMM'"() {bLoopAnimation=false, sequence=&/ATV_Driver_Collision_S_CMM})]
 }
 JumpLand {
  path=VehicleFinalPoseOverride.AnimGraphNode_StateMachine_1.FullBodyAdditive.AnimStateNode_3.JumpLand
  nodes [
   n1 = Get LandStartPosition() -> (LandStartPosition:float)
   Output Animation Pose(Result=Apply Additive!(
     Base="Sequence Player (additive) 'ATV_Driver_Land_CMM'"!(StartPosition:float=@n1) -> (Pose) {bLoopAnimation=false, sequence=&/ATV_Driver_Land_CMM}
     Additive=Use cached pose 'JumpLeanPose'()
     Alpha:float
    ) -> (Pose) {alphaBoolBlend {blendOption=Linear}})
  ]
 }
 Default {
  path=VehicleFinalPoseOverride.AnimGraphNode_StateMachine_1.FullBodyAdditive.AnimStateNode_4.Default
  nodes [Output Animation Pose(Result=Blend Poses by bool(
     BlendPose_0="Sequence Player (additive) 'ATV_Driver_Idle_Fast'"() {groupName=Idle, method=SyncGroup, sequence=&/ATV_Driver_Idle_Fast}
     BlendPose_1="Sequence Player (additive) 'ATV_Driver_Idle'"() {groupName=Idle, method=SyncGroup, sequence=&/ATV_Driver_Idle}
     BlendTime_0=0.25
     BlendTime_1=0.25
     bActiveValue=Get bIsDrivingFast() -> (bIsDrivingFast:bool)
    ) {blendType=Linear})]
 }
 Jump {
  path=VehicleFinalPoseOverride.AnimGraphNode_StateMachine_1.FullBodyAdditive.AnimStateNode_5.Jump
  nodes [
   n1 = Apply Additive(
    Base="Sequence Player (additive) 'ATV_Driver_Fall_CMM'"!(PlayRate:float=Get JumpPlayRate() -> (JumpPlayRate:float)) -> (Pose) {bLoopAnimation=false, sequence=&/ATV_Driver_Fall_CMM}
    Additive=Use cached pose 'JumpLeanPose'()
   ) {alphaBoolBlend {blendOption=Linear}}
   n2 = "Sequence Player (additive) 'ATV_Driver_Jump_Loop'"() {bLoopAnimation=false, sequence=&/ATV_Driver_Jump_Loop}
   AnimGraphNode_StateMachine_0 = JumpFlipAdditive() {is=title AnimGraphNode_StateMachine}
   Output Animation Pose(Result=Blend Poses by bool(
     BlendPose_0=@AnimGraphNode_StateMachine_0
     BlendPose_1=Apply Additive(Base=@n1, Additive=@n2) {alphaBoolBlend {blendOption=Linear}}
     BlendTime_0=0.3
     BlendTime_1=0.5
     bActiveValue=Get bIsJumpTrick() -> (bIsJumpTrick:bool)
    ) {blendType=Linear})
  ]
 }
 JumpFlipAdditive {
  entry=FlipStart
  path=VehicleFinalPoseOverride.AnimGraphNode_StateMachine_1.FullBodyAdditive.AnimStateNode_5.Jump.AnimGraphNode_StateMachine_0.JumpFlipAdditive
  ids=id
  nodes [
   AnimStateNode_0 = FlipStart() {is=title AnimStateNode}
   AnimStateNode_1 = FlipLoop() {is=title AnimStateNode}
   AnimStateNode_2 = FlipEnd() {is=title AnimStateNode}
   AnimStateTransitionNode_0 = AnimStateTransitionNode() {bAutomaticRuleBasedOnSequencePlayerInState=true}
   AnimStateTransitionNode_1 = AnimStateTransitionNode() {crossfadeDuration=0.15}
  ]
  links [
   {
    dst=FlipEnd
    rule=VehicleFinalPoseOverride.AnimGraphNode_StateMachine_1.FullBodyAdditive.AnimStateNode_5.Jump.AnimGraphNode_StateMachine_0.JumpFlipAdditive.AnimStateTransitionNode_1.Transition
    src=FlipLoop
    via=AnimStateTransitionNode_1
   }
   {
    dst=FlipLoop
    rule=VehicleFinalPoseOverride.AnimGraphNode_StateMachine_1.FullBodyAdditive.AnimStateNode_5.Jump.AnimGraphNode_StateMachine_0.JumpFlipAdditive.AnimStateTransitionNode_0.Transition
    src=FlipStart
    via=AnimStateTransitionNode_0
   }
  ]
 }
 FlipStart {
  path=VehicleFinalPoseOverride.AnimGraphNode_StateMachine_1.FullBodyAdditive.AnimStateNode_5.Jump.AnimGraphNode_StateMachine_0.JumpFlipAdditive.AnimStateNode_0.FlipStart
  nodes [Output Animation Pose(Result="Sequence Player (additive) 'ATV_Driver_JumpFlip_Start'"() {bLoopAnimation=false, sequence=&/ATV_Driver_JumpFlip_Start})]
 }
 FlipLoop {
  path=VehicleFinalPoseOverride.AnimGraphNode_StateMachine_1.FullBodyAdditive.AnimStateNode_5.Jump.AnimGraphNode_StateMachine_0.JumpFlipAdditive.AnimStateNode_1.FlipLoop
  nodes [Output Animation Pose(Result="Sequence Player (additive) 'ATV_Driver_JumpFlip'"() {sequence=&/ATV_Driver_JumpFlip})]
 }
 FlipEnd {
  path=VehicleFinalPoseOverride.AnimGraphNode_StateMachine_1.FullBodyAdditive.AnimStateNode_5.Jump.AnimGraphNode_StateMachine_0.JumpFlipAdditive.AnimStateNode_2.FlipEnd
  nodes [Output Animation Pose(Result="Sequence Player (additive) 'ATV_Driver_JumpFlip_Stop'"!(PlayRate:float=1.2) -> (Pose) {bLoopAnimation=false, sequence=&/ATV_Driver_JumpFlip_Stop})]
 }
 Transition {
  path=VehicleFinalPoseOverride.AnimGraphNode_StateMachine_1.FullBodyAdditive.AnimStateNode_5.Jump.AnimGraphNode_StateMachine_0.JumpFlipAdditive.AnimStateTransitionNode_0.Transition
  nodes [Result()]
 }
 Transition {
  path=VehicleFinalPoseOverride.AnimGraphNode_StateMachine_1.FullBodyAdditive.AnimStateNode_5.Jump.AnimGraphNode_StateMachine_0.JumpFlipAdditive.AnimStateTransitionNode_1.Transition
  nodes [Result(bCanEnterTransition=NOT Boolean(A=Get bIsJumpTrick()))]
 }
 Transition {path=VehicleFinalPoseOverride.AnimGraphNode_StateMachine_1.FullBodyAdditive.AnimStateTransitionNode_10.Transition, nodes [Result()]}
 Transition {
  path=VehicleFinalPoseOverride.AnimGraphNode_StateMachine_1.FullBodyAdditive.AnimStateTransitionNode_11.Transition
  nodes [Result(bCanEnterTransition=Break Fort Anim Input Velocity Impact!(FortAnimInput_VelocityImpact:"Fort Anim Input Velocity Impact Structure (by ref)"=Get VelocityImpact() -> (VelocityImpact:Fort Anim Input Velocity Impact Structure)) -> (
     DeltaVelocityThreshold:Vector=~
     ImpactScale:Vector=~
     ImpactLimitX:Input Range Structure
     ImpactLimitY:Input Range Structure
     ImpactLimitZ:Input Range Structure
     SpringInterpolatorX:Float RK4Spring Interpolator Structure
     SpringInterpolatorY:Float RK4Spring Interpolator Structure
     SpringInterpolatorZ:Float RK4Spring Interpolator Structure
     bTestVelocity:bool=~
     TestVelocity:Vector=~
     bIsForwardImpact:bool=~
     bIsBackwardImpact:bool=~
     bIsLeftImpact:bool=~
     bIsRightImpact:bool=~
     bIsUpImpact:bool=~
     bIsDownImpact:bool=~
    ) {is=title K2Node_BreakStruct}.bIsForwardImpact)]
 }
 Transition {path=VehicleFinalPoseOverride.AnimGraphNode_StateMachine_1.FullBodyAdditive.AnimStateTransitionNode_4.Transition, nodes [Result()]}
 Transition {
  path=VehicleFinalPoseOverride.AnimGraphNode_StateMachine_1.FullBodyAdditive.AnimStateTransitionNode_5.Transition
  nodes [Result(bCanEnterTransition=Get bIsJumpingAndNotRolling() -> (bIsJumpingAndNotRolling:bool))]
 }
 Transition {
  path=VehicleFinalPoseOverride.AnimGraphNode_StateMachine_1.FullBodyAdditive.AnimStateTransitionNode_6.Transition
  nodes [Result(bCanEnterTransition=Get bIsStopped())]
 }
 Transition {
  path=VehicleFinalPoseOverride.AnimGraphNode_StateMachine_1.FullBodyAdditive.AnimStateTransitionNode_7.Transition
  nodes [Result(bCanEnterTransition=Break Fort Anim Input Velocity Impact(FortAnimInput_VelocityImpact=Get VelocityImpact()) -> (
     DeltaVelocityThreshold=~
     ImpactScale=~
     bTestVelocity=~
     TestVelocity=~
     bIsForwardImpact=~
     bIsBackwardImpact=~
     bIsLeftImpact=~
     bIsRightImpact=~
     bIsUpImpact=~
     bIsDownImpact=~
    ).bIsBackwardImpact)]
 }
 Transition {path=VehicleFinalPoseOverride.AnimGraphNode_StateMachine_1.FullBodyAdditive.AnimStateTransitionNode_8.Transition, nodes [Result()]}
 Transition {
  path=VehicleFinalPoseOverride.AnimGraphNode_StateMachine_1.FullBodyAdditive.AnimStateTransitionNode_9.Transition
  nodes [Result(bCanEnterTransition=Get bIsOnGroundAndNotRolling() -> (bIsOnGroundAndNotRolling:bool))]
 }
 VehicleFullBodyOverride {nodes [
   Input Pose() {name=InPoseFullBody}
   Output Pose() {layerGroup=VehicleOverride}
  ]}
 VehicleLowerBodyOverride {nodes [
   Input Pose() {name=InPoseLowerBodyDefault}
   Input Pose() {name=InPoseUpperBody}
   Output Pose() {layerGroup=VehicleOverride}
  ]}
 VehicleSplitBodyOverride {nodes [
   Input Pose() {name=InPoseSplitBody}
   Input Pose() {name=InPoseUpperAndLowerBody}
   Input Pose() {name=InPoseUpperAndLowerBodyRemoveRoot}
   Output Pose() {layerGroup=Vehicle Override}
  ]}
]
variables []
